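# Parent material for Swirl vehicle wheel cosmetics: an unlit, masked, two-sided animated effect for skeletal meshes.
# Panned flow and noise textures, UV gradients, Fresnel and sine-driven spots shape the opacity and highlight masks.
# Emissive blends a custom color toward a hue-shifted variant, with optional desaturated wave highlights.
# Sine waves wobble vertices, optionally masked by vertex color, and the result passes through vehicle damage and de-rez transition functions.
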Material /VehicleCosmetics/Wheels/Swirl/Material/M_ParentWheel_Swirl
graph {
 properties {
  bAutomaticallySetUsageInEditor=false
  bUseMaterialAttributes=true
  bUsedWithSkeletalMesh=true
  blendMode=BLEND_Masked
  shadingModel=MSM_Unlit
  twoSided=true
  usedShadingModels=Unlit
 }
 nodes [
  n1 = NamedRerouteUsage()
  n2 = Clamp(None=NamedRerouteUsage())
  n3 = ConstantBiasScale(None=ComponentMask(None=TextureCoordinate()) {g=true}) {bias=0, scale=1.5}
  n4 = Multiply(
   A=ConstantBiasScale(None=NamedRerouteUsage()) {bias=0, scale=0.6}
   B=MaterialFunctionCall!(Normal, Exponent=Constant() {r=1}) {materialFunction=/VehicleCosmetics/Wheels/Cola_Holographic/Materials/MF_Fresnel_Classic}.Result
  )
  n5 = ScalarParameter() {defaultValue=0.1, parameterName=FresnelCutoutBias}
  n6 = Clamp(None=ConstantBiasScale(None=Subtract(A=@n4, B=@n5)) {bias=0, scale=10})
  n7 = Subtract(A=ComponentMask(None=TextureCoordinate()) {g=true}, B=ScalarParameter() {defaultValue=0.2, parameterName=GradientBias})
  n8 = Clamp(None=ConstantBiasScale(None=@n7) {bias=0, scale=1})
  n9 = VectorParameter() {parameterName=FlowMap_Tiling, defaultValue=<rgba 1 1 0 1>}
  n10 = AppendVector(A=@n9.R, B=@n9.R)
  n11 = TextureCoordinate()
  n12 = Time()
  n13 = ScalarParameter() {defaultValue=1, parameterName=FlowMapPanSpeed}
  n14 = TextureSampleParameter2D(UVs=Panner(Coordinate=Multiply(A=@n10, B=@n11), Time=Multiply(A=@n12, B=@n13)) {speedY=0.4}) {parameterName=FlowMap, samplerType=SAMPLERTYPE_Normal, texture=/VehicleCosmetics/Wheels/_SharedTextures/FX/Flow_03_nomips}
  n15 = ConstantBiasScale(None=@n14.R) {bias=0, scale=2}
  n16 = Clamp(None=ConstantBiasScale(None=Abs(None=@n14.G)) {bias=0.05, scale=1})
  n17 = Multiply(A=@n8, B=Add(A=@n15, B=@n16))
  n18 = Clamp(None=ConstantBiasScale(None=Clamp(None=@n17)) {bias=0.5, scale=1})
  n19 = OneMinus(None=MaterialFunctionCall(Exponent=Constant() {r=0.25}) {materialFunction=&MF_Fresnel_Classic}.Result)
  n20 = Clamp(None=ConstantBiasScale(None=Multiply(A=@n19, B=@n8)) {bias=-0.1, scale=2})
  n21 = TextureCoordinate()
  n22 = ComponentMask(None=@n21) {g=true}
  n23 = ComponentMask(None=@n21) {r=true}
  n24 = VectorParameter() {parameterName=AnisoTiling, defaultValue=<rgba 5 1.5 0 1>}
  n25 = Add(
   A=Multiply(A=AppendVector(A=@n22, B=@n23), B=ComponentMask(None=@n24.RGB) {g=true, r=true})
   B=ConstantBiasScale(None=NamedRerouteUsage()) {bias=0}
  )
  n26 = TextureSampleParameter2D(UVs=Panner(Coordinate=@n25) {speedX=1, speedY=-0.025}) {parameterName=Aniso, texture=/VehicleCosmetics/Textures/UniversalTextures/Universal_Noise_42_RGB}
  n27 = Multiply(
   A=TextureSampleParameter2D(UVs=Panner(Coordinate=@n25) {speedX=0.5, speedY=0.05}) {parameterName=Aniso, texture=&Universal_Noise_42_RGB}.G
   B=@n26.B
  )
  n28 = ConstantBiasScale(None=@n27) {bias=-0.05, scale=2}
  n29 = Add(A=@n7, B=@n28)
  n30 = Add(A=Clamp(None=Add(A=@n18, B=@n20)), B=Clamp(None=ConstantBiasScale(None=@n29) {bias=0, scale=3}, Min=Constant()))
  n31 = Constant() {r=0.1}
  n32 = Clamp(None=ConstantBiasScale(None=MaterialFunctionCall(Exponent=@n31) {materialFunction=&MF_Fresnel_Classic}.Result) {bias=-0.5, scale=10})
  n33 = Multiply(A=ComponentMask(None=TextureCoordinate() {uTiling=0.5}) {r=true}, B=ScalarParameter() {defaultValue=10, parameterName=OffsetXTiles})
  n34 = Multiply(A=Time(), B=ScalarParameter() {defaultValue=-0.75, parameterName=OffsetXPanSpeed})
  n35 = NamedRerouteDeclaration(None=Sine(None=Add(A=@n33, B=@n34))) {name=XTiles_Sine, nodeColor=<rgba 0 0.1328 1 1>}
  n36 = ComponentMask(None=TextureCoordinate() {uTiling=0.5}) {g=true}
  n37 = ScalarParameter() {defaultValue=7, parameterName=OffsetYTiles}
  n38 = Time()
  n39 = ScalarParameter() {defaultValue=1, parameterName=OffsetYPanSpeed}
  n40 = Sine(None=Add(A=Multiply(A=@n36, B=@n37), B=Multiply(A=@n38, B=@n39)))
  n41 = Multiply(A=@n40, B=@n35)
  n42 = Desaturation(None=Multiply(A=@n35, B=@n41)) {luminanceFactors=<rgba 0.3 0.59 0.11 0>}
  n43 = NamedRerouteUsage()
  n44 = Multiply(A=ConstantBiasScale(None=Multiply(A=@n42, B=@n43)) {bias=0, scale=2}, B=NamedRerouteUsage())
  n45 = Clamp(None=ConstantBiasScale(None=Abs(None=@n44)) {bias=0, scale=4}) {maxDefault=3}
  n46 = Clamp(None=ConstantBiasScale(None=Multiply(A=@n32, B=@n45)) {bias=-0.5, scale=2})
  n47 = Clamp(None=ConstantBiasScale(None=NamedRerouteUsage()) {bias=0, scale=3})
  n48 = ConstantBiasScale(None=NamedRerouteUsage()) {bias=-0.15, scale=10}
  n49 = StaticSwitchParameter(True=Multiply(A=Multiply(A=@n46, B=@n47), B=Clamp(None=@n48)), False=Constant()) {defaultValue=true, parameterName=VertexSpotOpacityNoise}
  n50 = ConstantBiasScale(None=NamedRerouteDeclaration(None=Subtract(A=@n30, B=@n49)) {name=Black_Sub, nodeColor=<rgba 1 1 0 1>}) {bias=-0.5, scale=1}
  n51 = NamedRerouteUsage()
  n52 = OneMinus(None=Clamp(None=ConstantBiasScale(None=@n51) {bias=-0.15, scale=10}))
  n53 = Add(A=Max(A=Multiply(A=@n2, B=@n3), B=Subtract(A=@n50, B=@n6)), B=Clamp(None=ConstantBiasScale(None=@n52) {bias=-0.75, scale=1}))
  n54 = Constant()
  n55 = Multiply(A=ConstantBiasScale(None=NamedRerouteUsage()) {bias=0}, B=ConstantBiasScale(None=@n26.B) {bias=-0.05, scale=2})
  n56 = Clamp(None=StaticSwitchParameter(True=LinearInterpolate(A=@n53, B=@n54, Alpha=@n55), False=@n53) {parameterName=MaskByVertexColor})
  n57 = NamedRerouteUsage()
  MaterialFunctionCall(In=MaterialFunctionCall(Material=SetMaterialAttributes!(MaterialAttributes, Emissive Color=@n1, Opacity=@n56, World Position Offset=@n57) {attributeSetTypes [#1, #2, #3]}) {materialFunction=&MF_BaseVehicle_Damage}.Result) {materialFunction=&MF_VehicleDeRez_Transition}
  n58 = AppendVector(A=@n15, B=@n16)
  n59 = Constant() {r=0.5}
  n60 = AppendVector(A=Constant(), B=@n14.G)
  NamedRerouteDeclaration(None=Multiply(A=LinearInterpolate(A=@n60, B=@n58, Alpha=@n59), B=ScalarParameter() {defaultValue=0.3, parameterName=NoiseDistort})) {name=FlowDistort, nodeColor=<rgba 0 1 0.8125 1>}
  n61 = NamedRerouteUsage()
  n62 = Constant() {r=0.175}
  n63 = ComponentMask(None=TextureCoordinate()) {g=true}
  n64 = Multiply(
   A=OneMinus(None=MaterialFunctionCall(Exponent=@n62) {materialFunction=&MF_Fresnel_Classic}.Result)
   B=@n63
  )
  n65 = ConstantBiasScale(None=@n64) {bias=0, scale=2}
  n66 = Multiply(A=ConstantBiasScale(None=@n35) {bias=0, scale=0.15}, B=NamedRerouteUsage())
  n67 = VectorParameter() {parameterName=NoiseMap_Tiling, defaultValue=<rgba 3 0.75 0 1>}
  n68 = Panner(
   Coordinate=Multiply(A=AppendVector(A=@n67.R, B=@n67.R), B=TextureCoordinate())
   Time=Multiply(A=Time(), B=ScalarParameter() {defaultValue=1, parameterName=NoisePanSpeed})
  ) {speedY=0.2}
  n69 = Constant() {r=-1}
  n70 = TextureSampleParameter2D(UVs=Add(A=BumpOffset(Coordinate=@n68, Height=@n69), B=NamedRerouteUsage())) {parameterName=NoiseMask, texture=/VehicleCosmetics/Wheels/_SharedTextures/FX/Universal_Noise_27_RGB}
  n71 = ConstantBiasScale(None=NamedRerouteDeclaration(None=@n70.R) {name=TexR, nodeColor=<rgba 0 1 0.1562 1>}) {bias=0, scale=1}
  n72 = OneMinus(None=@n63)
  n73 = Clamp(None=ConstantBiasScale(None=@n72) {bias=-0.05, scale=2})
  n74 = Clamp(None=ConstantBiasScale(None=@n70.B) {bias=0, scale=0.4})
  n75 = ConstantBiasScale(None=@n64) {bias=0, scale=10}
  n76 = Subtract(A=Clamp(None=Add(A=@n65, B=@n66)), B=LinearInterpolate(A=Clamp(None=@n71), B=Multiply(A=@n73, B=@n74), Alpha=Clamp(None=@n75)))
  n77 = Clamp(None=Add(A=@n28, B=ConstantBiasScale(None=@n76) {bias=0, scale=1.2}))
  NamedRerouteDeclaration(None=Power(Base=Max(A=@n61, B=@n77), Exp=ScalarParameter() {defaultValue=2, parameterName=HighLighSharpness})) {name=HighColorMask, nodeColor=<rgba 1 0.75 0 1>}
  NamedRerouteDeclaration(None=@n27) {name=Texture, nodeColor=<rgba 1 0 0.1641 1>}
  n78 = ComponentMask(None=TextureCoordinate()) {g=true}
  n79 = Clamp(None=Abs(None=ConstantBiasScale(None=@n78) {bias=-0.5, scale=2}))
  n80 = Clamp(None=ConstantBiasScale(None=NamedRerouteUsage()) {bias=0.5, scale=1})
  n81 = Clamp(None=OneMinus(None=@n79))
  n82 = AppendVector(A=Multiply(A=@n41, B=@n79), B=Multiply(A=Multiply(A=@n40, B=@n80), B=@n81))
  n83 = Multiply(A=@n41, B=@n81)
  n84 = ScalarParameter() {defaultValue=1, parameterName=OfsetWobble_head}
  n85 = VertexColor()
  n86 = Add(
   A=ConstantBiasScale(None=NamedRerouteDeclaration(None=@n85.R) {name=VC_RC, nodeColor=<rgba 1 0.1875 0 1>}) {bias=0, scale=1}
   B=ConstantBiasScale(None=OneMinus(None=@n78)) {bias=0, scale=1}
  )
  n87 = Multiply(
   A=Multiply(A=AppendVector(A=@n82, B=@n83), B=LinearInterpolate(B=@n84, Alpha=@n78))
   B=ConstantBiasScale(None=Clamp(None=@n86)) {bias=0.02, scale=3}
  )
  n88 = ConstantBiasScale(None=NamedRerouteUsage()) {bias=0}
  NamedRerouteDeclaration(None=StaticSwitchParameter(True=LinearInterpolate(A=@n87, Alpha=@n88) {constB=0}, False=@n87) {parameterName=MaskByVertexColor}) {name=WPO, nodeColor=<rgba 0 0.4844 1 1>}
  NamedRerouteDeclaration(None=@n72) {name=TexCMask1-x, nodeColor=<rgba 0 0.8594 1 1>}
  n89 = ComponentMask(None=TextureCoordinate()) {g=true}
  n90 = ConstantBiasScale(None=OneMinus(None=@n89)) {bias=0, scale=3}
  n91 = Add(A=ConstantBiasScale(None=NamedRerouteUsage()) {bias=-0.05, scale=2}, B=Multiply(A=Clamp(None=@n90), B=NamedRerouteUsage()))
  NamedRerouteDeclaration(None=@n91) {name=add, nodeColor=<rgba 0.5 0 1 1>}
  n92 = TextureCoordinate()
  n93 = ScalarParameter() {defaultValue=5, parameterName=SparklingTiling}
  n94 = ScalarParameter() {defaultValue=1, parameterName=SparkleNoise}
  n95 = NamedRerouteUsage()
  n96 = BumpOffset(Coordinate=Add(A=Multiply(A=@n92, B=@n93), B=Multiply(A=@n94, B=@n95)), Height=Constant() {r=-1})
  n97 = Multiply(A=Time(), B=ScalarParameter() {defaultValue=1, parameterName=sparklepannerspeedParam})
  n98 = Multiply(
   A=TextureSample(UVs=Panner(Coordinate=@n96, Time=@n97) {speedX=-0.05, speedY=1.2}) {texture=/VehicleCosmetics/Wheels/_SharedTextures/FX/Universal_Noise_32_RGB}.B
   B=NamedRerouteUsage()
  )
  n99 = NamedRerouteUsage()
  NamedRerouteDeclaration(None=ConstantBiasScale(None=Multiply(A=@n98, B=@n99)) {bias=-0.005, scale=20}) {name=32, nodeColor=<rgba 0 1 0.1797 1>}
  n100 = Clamp(None=NamedRerouteUsage())
  n101 = Clamp(None=ConstantBiasScale(None=@n100) {bias=0.75, scale=1})
  n102 = NamedRerouteUsage()
  n103 = Multiply(
   A=VectorParameter() {parameterName=CustomColor, defaultValue=<rgba 0.801 0.026 0 1>}.RGB
   B=Clamp(None=ConstantBiasScale(None=@n102) {bias=0.25, scale=1})
  )
  n104 = ConstantBiasScale(None=Multiply(A=ConstantBiasScale(None=@n89) {bias=-0.6, scale=3}, B=@n100)) {bias=-0.1, scale=2}
  n105 = ConstantBiasScale(None=@n100) {bias=0, scale=0.3}
  n106 = LinearInterpolate(
   A=@n103
   B=MaterialFunctionCall!(Color=Multiply(A=@n101, B=@n103), Angle=ScalarParameter() {defaultValue=0.075, parameterName=HueShift}) {materialFunction=/VehicleCosmetics/Wheels/Concept_Inverted/Materials/MF_HUE_Shift}.Result
   Alpha=Multiply(A=Max(A=@n104, B=@n105), B=ScalarParameter() {defaultValue=1, parameterName=Highlight intensity})
  )
  n107 = ScalarParameter() {defaultValue=0.5, parameterName=WaveDesaturation}
  n108 = Multiply(A=Clamp(None=ConstantBiasScale(None=@n89) {bias=-0.1, scale=0.6}), B=@n91)
  n109 = Max(
   A=Multiply(A=ConstantBiasScale(None=@n108) {bias=-0.1, scale=2}, B=ScalarParameter() {defaultValue=3, parameterName=Wavehighlightpower})
   B=NamedRerouteUsage()
  )
  n110 = ConstantBiasScale(None=Max(A=@n108, B=NamedRerouteUsage())) {bias=0.9, scale=1}
  n111 = ConstantBiasScale(None=@n89) {bias=0, scale=0.07}
  n112 = Clamp(None=ConstantBiasScale(None=Multiply(A=@n110, B=@n111)) {bias=-0.055, scale=300})
  n113 = LinearInterpolate(
   A=@n106
   B=ConstantBiasScale(None=Desaturation(None=@n106, Fraction=@n107) {luminanceFactors=<rgba 0.3 0.59 0.11 1>}) {bias=0, scale=2}
   Alpha=Clamp(None=Max(A=@n109, B=@n112))
  )
  NamedRerouteDeclaration(None=Multiply(
    A=StaticSwitchParameter(True=@n113, False=@n106) {parameterName=WaveHighlights}
    B=ScalarParameter() {defaultValue=1, parameterName=Emissive}
   )) {name=emissive, nodeColor=<rgba 1 0 0.1406 1>}
  NamedRerouteDeclaration(None=@n70.G) {name=TexG, nodeColor=<rgba 0.4766 0 1 1>}
 ]
}
