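# Parent material for the Swirl vehicle wheel glow, a two-sided translucent unlit material for skeletal meshes.
# Panning noise on axis-swapped UVs is added to the V gradient, forming a mask that blends CustomColor with its hue-shifted variant, scaled by Emissive.
# The same mask drives opacity, faded out by a U-based ramp and red vertex color.
# The result passes through the shared vehicle damage and de-rez transition functions.
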Material /VehicleCosmetics/Wheels/Swirl/Material/M_Parent_Wheel_Swirl_Glow
graph {
 properties {
  bAutomaticallySetUsageInEditor=false
  bUseMaterialAttributes=true
  bUsedWithSkeletalMesh=true
  blendMode=BLEND_Translucent
  shadingModel=MSM_Unlit
  twoSided=true
  usedShadingModels=Unlit
 }
 nodes [
  n1 = VectorParameter() {parameterName=CustomColor, defaultValue=<rgba 0.7913 0.0296 0 1>}
  n2 = ScalarParameter() {defaultValue=0.075, parameterName=HueShift}
  n3 = ComponentMask(None=TextureCoordinate()) {g=true}
  n4 = TextureCoordinate() {uTiling=0.15, vTiling=0.1}
  n5 = AppendVector(A=ComponentMask(None=@n4) {g=true}, B=ComponentMask(None=@n4) {r=true})
  n6 = ConstantBiasScale(None=TextureSample(UVs=Panner(Coordinate=@n5) {speedX=0.5, speedY=-0.03}) {texture=/VehicleCosmetics/Wheels/Tondi_Inverted/Textures/Universal_Noise_42_RGB}.B) {bias=0, scale=1}
  n7 = Clamp(None=ConstantBiasScale(None=Add(A=@n3, B=@n6)) {bias=-0.55, scale=1})
  n8 = Multiply(
   A=LinearInterpolate(
    A=@n1.RGB
    B=MaterialFunctionCall!(Color=@n1.RGB, Angle=@n2) {materialFunction=/VehicleCosmetics/Wheels/Concept_Inverted/Materials/MF_HUE_Shift}.Result
    Alpha=@n7
   )
   B=ScalarParameter() {defaultValue=5, parameterName=Emissive}
  )
  n9 = TextureCoordinate()
  n10 = ConstantBiasScale(None=OneMinus(None=ComponentMask(None=@n9) {r=true})) {bias=-0.45, scale=10}
  n11 = ConstantBiasScale(None=VertexColor().R) {bias=0.1, scale=3}
  n12 = LinearInterpolate(A=Clamp(None=ConstantBiasScale(None=@n7) {bias=0, scale=1}), B=Constant(), Alpha=Max(A=Clamp(None=@n10), B=Clamp(None=@n11)))
  MaterialFunctionCall(In=MaterialFunctionCall(Material=SetMaterialAttributes!(MaterialAttributes, Emissive Color=@n8, Opacity=@n12) {attributeSetTypes [#1, #2]}) {materialFunction=&MF_BaseVehicle_Damage}.Result) {materialFunction=&MF_VehicleDeRez_Transition}
 ]
}
